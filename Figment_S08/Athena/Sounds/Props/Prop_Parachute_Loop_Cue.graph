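# Sound cue for a looping parachute prop in Figment_S08, built from third-person skydive parachute wind loops.
# A quality-level node picks between the standard wave for the first two levels and an "opt" variant for the last two, feeding a Modulator at the root.
# The cue lowers pitch and volume, plays in the in-game SFX class, and overrides attenuation with short-range falloff and listener focus that quiets non-focused playback.
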SoundCue /Figment_S08/Athena/Sounds/Props/Prop_Parachute_Loop_Cue
root=Modulator_0
nodes [
 {id=Modulator_0, children [QualityLevel_0]}
 {id=QualityLevel_0, children [WavePlayer_1,WavePlayer_1,WavePlayer_2,WavePlayer_2]}
 {id=WavePlayer_1, properties {bLooping=true, soundWaveAssetPtr=/Game/Athena/Sounds/Wind_Falling/sfx_parachute_skydive_loop_3p}}
 {id=WavePlayer_2, properties {bLooping=true, soundWaveAssetPtr=/Figment_S08/Athena/Sounds/Wind_Falling/sfx_parachute_skydive_loop_3p_opt}}
]
properties {
 bOverrideAttenuation=true
 maxDistance=1000
 pitchMultiplier=0.5
 soundClassObject=&Fort_SFX_InGame
 volumeMultiplier=0.7
 attenuationOverrides {
  bEnableFocusInterpolation=true
  bEnableListenerFocus=true
  falloffDistance=1000
  nonFocusVolumeAttenuation=0.2
  stereoSpread=512
  attenuationShapeExtents {x=0}
 }
}
